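# Unlit translucent material for Niagara mesh-particle clouds on the Full Moon backpack cosmetic.
# A panning world-aligned storm-cloud texture supplies normals for fake lighting, blending shadow and light colors against a parameterized light origin.
# A sphere-masked fake point light adds noise-tinted particle color, with strength and radius driven by dynamic parameters.
# Fresnel and particle alpha shape opacity with non-mobile depth fade and optional view-angle fade; the texture also modulates vertex-normal displacement.
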Material /BRCosmetics/Accessories/FORT_Backpacks/Backpack_M_MED_FullMoon/FX/M_Backpack_FullMoon_Clouds
graph {
 properties {
  bUsedWithNiagaraMeshParticles=true
  blendMode=BLEND_Translucent
  pixelDepthOffsetMode=PDOM_Legacy
  shadingModel=MSM_Unlit
  usedShadingModels=Unlit
 }
 nodes [
  n1 = DynamicParameter() {parameterIndex=1, paramNames [StrengthPointLight, PointLightRadius, HueShift, Hardness]}
  n2 = ParticleColor()
  n3 = Panner() {speedX=-0.08}
  n4 = Multiply(
   A=@n1.StrengthPointLight
   B=LinearInterpolate(
    A=Multiply(A=@n1.HueShift, B=@n2.RGB)
    B=@n2.RGB
    Alpha=TextureSample(UVs=@n3) {texture=&T_NoisePack_01}.G
   )
  )
  n5 = Subtract(A=WorldPosition().XYZ, B=ObjectPositionWS())
  n6 = SphereMask(
   A=@n5
   B=VectorParameter() {desc=Particle Backpack Offset, parameterName=Point Light Offset, defaultValue=<rgba 15 0 0 1>}.RGB
   Radius=@n1.PointLightRadius
   Hardness=@n1.PointLightRadius
  ) {hardnessPercent=0}
  n7 = ComponentMask(None=Add(
    A=VectorParameter() {desc=Particle Backpack Offset, parameterName=Position Offset, defaultValue=<rgba 15 0 0 1>}.RGB
    B=VectorParameter() {parameterName=LightOrigin}.RGB
   )) {b=true, g=true, r=true}
  n8 = DynamicParameter() {paramNames [PanningSpeedX, PanningSpeedY, WorldTileSize, WPO Amount]}
  n9 = Subtract(A=WorldPosition().XYZ, B=ObjectPositionWS())
  n10 = ComponentMask(None=VectorParameter() {parameterName=Speed, defaultValue=<rgba 0.008 0.0063 0 0>}.RGB) {g=true, r=true}
  n11 = DynamicParameter() {paramNames [PanningSpeedX, PanningSpeedY, WorldTileSize, WPO Amount]}
  n12 = Multiply(A=@n11.PanningSpeedX, B=@n11.PanningSpeedX)
  n13 = MaterialFunctionCall(
   Normal=MaterialFunctionCall!(
    TextureObject=TextureObject() {samplerType=SAMPLERTYPE_LinearColor, texture=/Game/VisualThreatManager/StormVisuals/LargeClouds/T_StormCloudMobile}
    TextureSize=MaterialFunctionCall(X=@n8.WorldTileSize, Y=@n8.WorldTileSize, Z=@n8.WorldTileSize) {materialFunction=&MakeFloat3}.Result
    WorldPosition=Transform(None=@n9) {transformSourceType=TRANSFORMSOURCE_World, transformType=TRANSFORM_ParticleWorld}
    Export Float 4
    In=Multiply(A=@n10, B=@n12)
    World Space Normal=VertexNormalWS()
    ProjectionTransitionContrast
   ) {materialFunction=/Game/Athena/SafeZone/WorldAlignedTexture_ScalePan}.XYZ Texture
   Flatness=ScalarParameter() {defaultValue=0.5, parameterName=FlattenDetailNormal}
  ) {materialFunction=&FlattenNormal}
  n14 = Transform(None=@n13.Result)
  n15 = LinearInterpolate(A=@n14, B=LinearInterpolate(A=@n14, B=VertexNormalWS()), Alpha=ScalarParameter() {defaultValue=0.1, parameterName=LerpVertNormal})
  n16 = ConstantBiasScale(None=DotProduct(A=Normalize(VectorInput=@n7), B=@n15)) {bias=0.5, scale=0.6}
  n17 = ScalarParameter() {defaultValue=2.4, parameterName=SoftenShadow}
  n18 = LinearInterpolate(
   A=VectorParameter() {parameterName=ShadowColorDay, defaultValue=<rgba 0.05 0.05 0.05 1>}.RGB
   B=VectorParameter() {parameterName=LightColorDay, defaultValue=<rgba 0.25 0.25 0.25 1>}.RGB
   Alpha=Clamp(None=Power(Base=@n16, Exp=@n17))
  )
  MaterialFunctionCall(Emissive=Add(A=LinearInterpolate(B=@n4, Alpha=@n6), B=@n18)) {materialFunction=&MF_FX_EmissiveAdjustments}
  n19 = ParticleColor()
  Multiply(A=@n18, B=@n19.RGB)
  n20 = Multiply(
   A=@n19.A
   B=MaterialFunctionCall(
    Brightness=ScalarParameter() {defaultValue=1, parameterName="Brightness (S)"}
    Exponent=ScalarParameter() {defaultValue=3.7673, parameterName="Exponent (S)"}
    Vertex / Pixel Normal=@n15
   ) {materialFunction=&MF_Fresnel_FX}.Fresnel
  )
  n21 = ScalarParameter() {defaultValue=5, parameterName=FadeDistance}
  n22 = ScalarParameter() {parameterName=ViewAngleFade}
  Multiply(
   A=ShadingPathSwitch(Default=DepthFade(Opacity=@n20, FadeDistance=@n21), Mobile=@n20)
   B=LinearInterpolate(
    B=MaterialFunctionCall("FadeSensitivity (optional)"=@n22) {materialFunction=&FadeBasedOnViewAngle}.Result
    Alpha=ScalarParameter() {parameterName=UseViewAngleFade}
   ) {constA=1}
  )
  n23 = ComponentMask(None=@n13.Result) {r=true}
  n24 = ComponentMask(None=@n13.Result) {g=true}
  Multiply(
   A=LinearInterpolate(B=DynamicParameter() {paramNames [PanningSpeedX, PanningSpeedY, WorldTileSize, WPO Amount]}.WPO Amount, Alpha=Add(A=@n23, B=@n24)) {constB=0}
   B=VertexNormalWS()
  )
  Transform(None=@n5) {transformSourceType=TRANSFORMSOURCE_World, transformType=TRANSFORM_Local}
 ]
}
