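# Niagara effect for the FullMoon backpack cosmetic: a glowing moon with clouds and twinkling stars.
# A system Timeline drives looping curves that shift moon and cloud tint between pale yellow and red-orange, animate moon material parameters and sway the middle cloud.
# Moon_Sprite pairs a mesh with a light in the particle color; Clouds and Stars each use three renderers, with per-renderer position attributes.
# The Darkness sprite emitter is disabled.
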NiagaraSystem /BRCosmetics/Accessories/FORT_Backpacks/Backpack_M_MED_FullMoon/FX/NS_Backpack_FullMoon
dependencies {
 usedDataInterfaces [&NiagaraDataInterfaceCurve, &NiagaraDataInterfaceColorCurve]
 usedDynamicInputs [&SetBoolByFloatComparison]
 usedModules [&SystemState, &Timeline]
}
emitters {
 Clouds {
  emitterUpdateScript [
   &EmitterState(Life Cycle Mode=::NewEnumerator1, Loop Behavior=::NewEnumerator1, Loop Duration Mode=::NewEnumerator1, UseLoopDelay=false)
   &SpawnBurst_Instantaneous(Use Spawn Probability=false, UseLoopCountLimit=false)
   /SetVariables(
    Emitter.FlattenNormal=0.4
    Emitter.LerpVertNormal=0.1
    Emitter.Light Source:Vector3f=/Niagara/DynamicInputs/Helpers/SimulationPosition()
    Emitter.MIddle Cloud Point Light Position:Vector3f=&MakeVector(
     X=&FloatFromCurve(
      FloatCurve=DataInterface() {propertyValues {Curve {keys [
          {arriveTangent=34.2303, interpMode=RCIM_Cubic, leaveTangent=34.2303, tangentMode=RCTM_User, time=0, value=-20}
          <at 1 0 RCIM_Cubic>
          {arriveTangent=-25.9105, interpMode=RCIM_Cubic, leaveTangent=-25.9105, tangentMode=RCTM_User, time=2, value=-20}
         ]}}}
      CurveIndex=$System.Timeline.Time
     )
     Y=0
     Z=0
    )
    Emitter.Point Light Position:Vector3f=<0 -2 0>
    Emitter.Position Offset:Vector3f=<-20 0 100>
    Emitter.SoftenShadow=1.25
   ) {isSetParameters=true}
  ]
  particleSpawnScript [&InitializeParticle(
    Color=<rgba 3 3 3 1>
    Mesh Scale=<1.1 1.1 1.1>
    Mesh Scale Mode=::NewEnumerator4
    Position Offset Coordinate Space=::Local
    UsePositionOffset=true
    Write Material Random=true
   )]
  particleUpdateScript [
   &ParticleState(Kill Particles When Lifetime Has Elapsed=false, Let Infinitely Lived Particles Die When Emitter Deactivates=true)
   &DynamicMaterialParameters(
    Index 0 Param 1=2
    Index 0 Param 2=-2
    Index 0 Param 3=15
    Index 0 Param 4=2
    Index 1 Param 1=0.25
    Index 1 Param 2=15
    Index 1 Param 3=0.25
    Module.Param0WriteEnabled=true
    Module.Param0WriteEnabled1=true
    Module.Param1WriteEnabled=true
    Module.Param1WriteEnabled1=true
    Module.Param2WriteEnabled=true
    Module.Param2WriteEnabled1=true
    Module.Param3WriteEnabled=true
    Module.Param3WriteEnabled1=true
    Write Parameter Index 1=true
   )
   /SetVariables(
    Particles.Bottom Cloud Position:Vector3f=&Add_Vector(A=Linked() {linkedVariable {name=Particles.Initial.Position, type=NiagaraPosition}}, B=<0 0 0>)
    Particles.MIddle Cloud Position:Vector3f=&Add_Vector(
     A=Linked() {linkedVariable {name=Particles.Initial.Position, type=NiagaraPosition}}
     B=&Lerp_Vector(
      A=<0 10 2>
      B=<0 0 0>
      Alpha=&FloatFromCurve(
       FloatCurve=DataInterface() {propertyValues {Curve {keys [<at 0 0 RCIM_Cubic>,<at 1 1 RCIM_Cubic>,<at 2 0 RCIM_Cubic>]}}}
       CurveIndex=$System.Timeline.Time
      )
     )
    )
    Particles.MIddleCloudDynamicParameters:Vector4f=<0.25 15 0.25 0>
    Particles.Top Cloud Position:Vector3f=&Add_Vector(A=Linked() {linkedVariable {name=Particles.Initial.Position, type=NiagaraPosition}}, B=<0 0 0>)
   ) {isSetParameters=true}
   /Niagara/Modules/Update/Orientation/V2/OrientMeshToVector(Look At Direction:Vector3f=<1 0 0>, Override Quat=false, UseOptionalSideAxis=false, UseOptionalUpAxis=false)
   &ScaleColor(
    Scale RGB=&MakeVectorFromLinearColorRGB(Color=&ColorFromCurve(
      ColorCurve=DataInterface() {propertyValues {
        AlphaCurve {keys []}
        BlueCurve {keys [
          {interpMode=RCIM_Cubic, leaveTangentWeight=0.0833, tangentMode=RCTM_User, time=0, value=0.5}
          {arriveTangentWeight=0.1179, interpMode=RCIM_Cubic, leaveTangentWeight=0.1179, tangentMode=RCTM_User, time=0.25, value=0}
          {arriveTangentWeight=0.0833, interpMode=RCIM_Cubic, leaveTangentWeight=0.0833, tangentMode=RCTM_User, time=0.5, value=0}
          {arriveTangentWeight=0.1179, interpMode=RCIM_Cubic, leaveTangentWeight=0.1179, tangentMode=RCTM_User, time=0.75, value=0}
          {arriveTangentWeight=0.0833, interpMode=RCIM_Cubic, tangentMode=RCTM_User, time=1, value=0.5}
         ]}
        GreenCurve {keys [
          <at 0 0.9282 RCIM_Cubic>
          {arriveTangentWeight=0.1684, interpMode=RCIM_Cubic, leaveTangentWeight=0.1684, tangentMode=RCTM_User, time=0.25, value=0.5029}
          <at 0.5 0.05 RCIM_Cubic>
          {arriveTangentWeight=0.1684, interpMode=RCIM_Cubic, leaveTangentWeight=0.1684, tangentMode=RCTM_User, time=0.75, value=0.5029}
          <at 1 0.9282 RCIM_Cubic>
         ]}
        RedCurve {keys [<at 0 1 RCIM_Cubic>,<at 0.25 1 RCIM_Cubic>,<at 0.5 1 RCIM_Cubic>,<at 0.75 1 RCIM_Cubic>,<at 1 1 RCIM_Cubic>]}
       }}
      CurveIndex=$System.Timeline.NormalizedTime
     ))
    ScaleA=true
    ScaleRGB=true
   )
  ]
  rendererClasses [&NiagaraMeshRendererProperties]
  renderers [
   {rendererClass=&NiagaraMeshRendererProperties, rendererIndex=0}
   {rendererClass=&NiagaraMeshRendererProperties, rendererIndex=1}
   {rendererClass=&NiagaraMeshRendererProperties, rendererIndex=2}
  ]
 }
 Darkness {
  bEnabled=false
  emitterUpdateScript [
   &EmitterState(Life Cycle Mode=::NewEnumerator1, Loop Behavior=::NewEnumerator1, Loop Duration Mode=::NewEnumerator1, UseLoopDelay=false)
   &SpawnBurst_Instantaneous(Use Spawn Probability=false, UseLoopCountLimit=false)
  ]
  particleSpawnScript [
   &InitializeParticle(
    Color=<rgba 0 0 0 1>
    Lifetime=0
    Sprite Size Mode=::NewEnumerator3
    Uniform Sprite Size=50
    UsePositionOffset=false
    Write Material Random=true
   )
   &SpriteFacingAndAlignment(WriteAlignment=false, WriteFacing=true)
   &DynamicMaterialParameters(
    Index 0 Param 1=1
    Module.Param0WriteEnabled=true
    Module.Param1WriteEnabled=true
    Module.Param2WriteEnabled=true
    Module.Param3WriteEnabled=true
   )
  ]
  particleUpdateScript [&ParticleState(Kill Particles When Lifetime Has Elapsed=false, Let Infinitely Lived Particles Die When Emitter Deactivates=true)]
  rendererClasses [&NiagaraSpriteRendererProperties]
  renderers [{rendererClass=&NiagaraSpriteRendererProperties}]
 }
 Moon_Sprite {
  emitterUpdateScript [
   &EmitterState(Life Cycle Mode=::NewEnumerator1, Loop Behavior=::NewEnumerator1, Loop Duration Mode=::NewEnumerator1, UseLoopDelay=false)
   &SpawnBurst_Instantaneous(Use Spawn Probability=false, UseLoopCountLimit=false)
  ]
  particleSpawnScript [&InitializeParticle(Color=<rgba 3 3 3 1>, Sprite Size Mode=::NewEnumerator3, Uniform Sprite Size=25, UsePositionOffset=true, Write Material Random=true)]
  particleUpdateScript [
   &ParticleState(Kill Particles When Lifetime Has Elapsed=false, Let Infinitely Lived Particles Die When Emitter Deactivates=true)
   &DynamicMaterialParameters(
    Index 0 Param 1=0.333
    Index 0 Param 2=&FloatFromCurve(
     FloatCurve=DataInterface() {propertyValues {Curve {keys [<at 0 0.2462 RCIM_Constant>,<at 0.05 0.2462>,<at 0.45 0.1068 RCIM_Constant>,<at 0.55 0.1068>,<at 0.95 -0.1788 RCIM_Constant>,<at 1.05 -0.1788>,<at 1.45 0.1068 RCIM_Constant>,<at 1.55 0.1068>,<at 1.95 0.2462 RCIM_Constant>]}}}
     CurveIndex=$System.Timeline.Time
    )
    Index 0 Param 3=&FloatFromCurve(
     FloatCurve=DataInterface() {propertyValues {Curve {keys [<at 0 0.3758 RCIM_Constant>,<at 0.05 0.3758>,<at 0.45 0.0848 RCIM_Constant>,<at 0.55 0.0848>,<at 0.95 -0.2424 RCIM_Constant>,<at 1.05 -0.2424>,<at 1.45 0.0848 RCIM_Constant>,<at 1.55 0.0848>,<at 1.95 0.3758 RCIM_Constant>]}}}
     CurveIndex=$System.Timeline.Time
    )
    Index 0 Param 4=0.5
    Module.Param0WriteEnabled=true
    Module.Param1WriteEnabled=true
    Module.Param2WriteEnabled=true
    Module.Param3WriteEnabled=true
   )
   &ScaleColor(
    Scale RGB=&MakeVectorFromLinearColorRGB(Color=&ColorFromCurve(
      ColorCurve=DataInterface() {propertyValues {
        AlphaCurve {keys []}
        BlueCurve {keys [
          {interpMode=RCIM_Cubic, leaveTangentWeight=0.0833, tangentMode=RCTM_User, time=0, value=0.5}
          {arriveTangentWeight=0.1179, interpMode=RCIM_Cubic, leaveTangentWeight=0.1179, tangentMode=RCTM_User, time=0.25, value=0}
          {arriveTangentWeight=0.0833, interpMode=RCIM_Cubic, leaveTangentWeight=0.0833, tangentMode=RCTM_User, time=0.5, value=0.05}
          {arriveTangentWeight=0.1179, interpMode=RCIM_Cubic, leaveTangentWeight=0.1179, tangentMode=RCTM_User, time=0.75, value=0}
          {arriveTangentWeight=0.0833, interpMode=RCIM_Cubic, tangentMode=RCTM_User, time=1, value=0.5}
         ]}
        GreenCurve {keys [
          <at 0 0.9282 RCIM_Cubic>
          {arriveTangentWeight=0.1684, interpMode=RCIM_Cubic, leaveTangentWeight=0.1684, tangentMode=RCTM_User, time=0.25, value=0.5029}
          <at 0.5 0.2083 RCIM_Cubic>
          {arriveTangentWeight=0.1684, interpMode=RCIM_Cubic, leaveTangentWeight=0.1684, tangentMode=RCTM_User, time=0.75, value=0.5029}
          <at 1 0.9282 RCIM_Cubic>
         ]}
        RedCurve {keys [<at 0 1 RCIM_Cubic>,<at 0.25 1 RCIM_Cubic>,<at 0.5 1 RCIM_Cubic>,<at 0.75 1 RCIM_Cubic>,<at 1 1 RCIM_Cubic>]}
       }}
      CurveIndex=$System.Timeline.NormalizedTime
     ))
    ScaleA=true
    ScaleRGB=true
   )
   &FN_Light_Attributes(
    Color=$Particles.Color
    Create Color Attribute=true
    Create Exponent Attribute=true
    Create Position Attribute=true
    Create Radius Attribute=true
    Exponent=2
    Position=<20 -2 0>
    Radius=30
    Scale Color=5
   )
  ]
  rendererClasses [&NiagaraMeshRendererProperties, &NiagaraLightRendererProperties]
  renderers [
   {rendererClass=&NiagaraMeshRendererProperties, rendererIndex=0}
   {rendererClass=&NiagaraLightRendererProperties, rendererIndex=1}
  ]
 }
 Stars {
  emitterUpdateScript [
   &EmitterState()
   &SpawnRate(SpawnRate=2, Use Spawn Probability=false)
  ]
  particleSpawnScript [&InitializeParticle(
    Color=<rgba 1 0.9641 0.75 1>
    Lifetime Mode=::NewEnumerator1
    Sprite Size Mode=::NewEnumerator3
    UsePositionOffset=false
    Write Material Random=true
   )]
  particleUpdateScript [
   &ParticleState()
   /SetVariables(
    Particles.Star A Position:Vector3f=<0 -2 15>
    Particles.Star B Position:Vector3f=<2 5 -13>
    Particles.Star B Size:Vector2f=&Multiply_Vector2D(
     A=$Particles.Initial.SpriteSize
     B=&Vector2DFromFloat(Value=&FloatFromCurve(FloatCurve=DataInterface() {propertyValues {Curve {keys [
           {interpMode=RCIM_Cubic, leaveTangentWeight=0.9072, tangentMode=RCTM_User, time=0, value=0}
           {arriveTangentWeight=0.2716, interpMode=RCIM_Cubic, leaveTangentWeight=0.6337, tangentMode=RCTM_User, time=0.62, value=1}
           {arriveTangent=-0.0405, interpMode=RCIM_Cubic, leaveTangent=-0.0405, tangentMode=RCTM_User, time=1, value=0}
          ]}}}))
    )
    Particles.Star C Position:Vector3f=<0 -10 10>
    Particles.Star C Size:Vector2f=&Multiply_Vector2D(
     A=$Particles.Initial.SpriteSize
     B=&Vector2DFromFloat(Value=&FloatFromCurve(FloatCurve=DataInterface() {propertyValues {Curve {keys [
           {interpMode=RCIM_Cubic, leaveTangentWeight=0.9072, tangentMode=RCTM_User, time=0, value=0}
           {arriveTangentWeight=0.2716, interpMode=RCIM_Cubic, leaveTangentWeight=0.6337, tangentMode=RCTM_User, time=0.475, value=1}
           {arriveTangent=-0.0405, interpMode=RCIM_Cubic, leaveTangent=-0.0405, tangentMode=RCTM_User, time=1, value=0}
          ]}}}))
    )
   ) {isSetParameters=true}
   &ScaleColor(
    Scale Alpha=&FloatFromCurve(
     FloatCurve=DataInterface() {propertyValues {Curve {keys [
         {interpMode=RCIM_Cubic, tangentMode=RCTM_User, time=0, value=0}
         {interpMode=RCIM_Cubic, tangentMode=RCTM_User, time=0.5, value=1}
         {tangentMode=RCTM_User, time=1, value=0}
        ]}}}
     Scale Curve=&RandomRangeFloat(Minimum=0.1, Maximum=0.25)
    )
    Scale RGB=<10 10 10>
    ScaleA=true
    ScaleRGB=true
   )
   &ScaleSpriteSize(Uniform Curve Sprite Scale=DataInterface() {propertyValues {Curve {keys [
        {arriveTangent=4.2744, interpMode=RCIM_Cubic, leaveTangent=4.2744, tangentMode=RCTM_User, time=0, value=0}
        {arriveTangentWeight=0.2716, interpMode=RCIM_Cubic, leaveTangentWeight=0.6337, tangentMode=RCTM_User, time=0.3, value=0.9847}
        {arriveTangent=-0.0405, interpMode=RCIM_Cubic, leaveTangent=-0.0405, tangentMode=RCTM_User, time=1, value=0}
       ]}}})
  ]
  rendererClasses [&NiagaraSpriteRendererProperties]
  renderers [
   {rendererClass=&NiagaraSpriteRendererProperties}
   {rendererClass=&NiagaraSpriteRendererProperties, rendererIndex=1}
   {rendererClass=&NiagaraSpriteRendererProperties, rendererIndex=2}
  ]
 }
}
summary {
 systemName=NS_Backpack_FullMoon
 emitters [
  {emitterName=Moon_Sprite, rendererClasses [&NiagaraMeshRendererProperties, &NiagaraLightRendererProperties]}
  {emitterName=Clouds, rendererClasses [&NiagaraMeshRendererProperties]}
  {bEnabled=false, emitterName=Darkness, rendererClasses [&NiagaraSpriteRendererProperties]}
  {emitterName=Stars, rendererClasses [&NiagaraSpriteRendererProperties]}
 ]
 userVariables []
}
